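# Character material function that optionally hides mesh elements by vertex color, editing World Position Offset rather than opacity despite its name.
# It passes the incoming material's WPO through the shared Fortnite_HideElementsWithVertexColor_2022 function and writes it back into the attributes.
# Static switch "Use Vertex Colors for Mask" enables this.
# When "OnlyIncludeVertexColorHideForHolo" is on, a zero-alpha lerp leaves the original WPO, so nothing is hidden here.
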MaterialFunction /Game/Global/Materials/Character_Parent_2023/Functions/BaseFN_CharacterOpacity_2023
graph {nodes [
  n1 = GetMaterialAttributes(None=FunctionInput() {inputName=Material, inputType=FunctionInput_MaterialAttributes}) {attributeGetTypes [#1]}
  n2 = MaterialFunctionCall!(WPO=@n1.WorldPositionOffset) {materialFunction=/Game/Packages/Fortress_SharedMaterials/Base_Material_Functions/Fortnite_HideElementsWithVertexColor_2022}
  n3 = StaticSwitchParameter(True=LinearInterpolate(A=@n1.WorldPositionOffset, B=@n2.Result) {constAlpha=0}, False=@n2.Result) {group=Opacity Mask, parameterName=OnlyIncludeVertexColorHideForHolo}
  FunctionOutput(None=SetMaterialAttributes!(
    MaterialAttributes=@n1.MaterialAttributes
    WorldPositionOffset=StaticSwitchParameter(True=@n3, False=@n1.WorldPositionOffset) {group=03_Opacity Mask, parameterName=Use Vertex Colors for Mask}
   ) {attributeSetTypes [#1]})
 ]}
