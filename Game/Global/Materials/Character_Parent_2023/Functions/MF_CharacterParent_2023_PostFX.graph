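# Material function for the 2023 character parent material that layers shared post effects onto incoming material attributes.
# It chains rim lighting, cloth-sim WPO, tech-art overrides, opacity, de-rez transition, hit glow and cinematic post FX.
# Named reroutes distribute inputs, including a packed AO/GMap/Skin mask, with Fresnel shared by de-rez and hit glow.
# Pre-skinned position defaults to a stage switch, using a vertex interpolator for pixels; cloth sim's skin mask is constant.
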MaterialFunction /Game/Global/Materials/Character_Parent_2023/Functions/MF_CharacterParent_2023_PostFX
graph {
 properties {bExposeToLibrary=true, libraryCategoriesText [TechArt]}
 nodes [
  n1 = NamedRerouteUsage()
  n2 = MaterialFunctionCall!(
   Material=MaterialFunctionCall!(
    Material=NamedRerouteUsage()
    Pre Skinned Local Position=NamedRerouteUsage()
    AO=NamedRerouteUsage()
    Gmap=NamedRerouteUsage()
    Distance Blend=NamedRerouteUsage()
   ) {materialFunction=/Game/Packages/Fortress_SharedMaterials/Base_Material_Functions/MF_RimV3}.Result
   SkinFX_Mask=Constant()
  ) {materialFunction=/Game/Global/MaterialFunctions/MF_ClothSimWPO_v2}
  n3 = MaterialFunctionCall!(
   Fresnel=@n1
   Material=MaterialFunctionCall!(Material=MaterialFunctionCall(In=@n2.Result) {materialFunction=&MF_TechArtOverrides}.Result) {materialFunction=&/BaseFN_CharacterOpacity_2023}.Result
  ) {materialFunction=/Game/Packages/Fortress_SharedMaterials/Base_Material_Functions/CharacterDeRez_Transition}
  FunctionOutput(None=MaterialFunctionCall(In=MaterialFunctionCall!(Material=@n3.Result, Fresnel=@n1) {materialFunction=/Game/Packages/Fortress_SharedMaterials/Base_Material_Functions/BaseCharacterShaderFunctions/MF_HitGlow}.Result) {materialFunction=&CINE_MF_CharPostFX}.Result)
  n4 = MaterialFunctionCall(Float3=FunctionInput(Preview=Constant()) {inputName=M, sortPriority=1, previewValue=<1 0 0 1>}) {desc="R = AO\r\nG = GMap\r\nB = Skin", materialFunction=&BreakOutFloat3Components}
  NamedRerouteDeclaration(None=@n4.R) {name=Ambient Occlusion, nodeColor=<rgba 0.9453 1 0 1>}
  NamedRerouteDeclaration(None=@n4.G) {name=GMap, nodeColor=<rgba 0 0.7292 1 1>}
  NamedRerouteDeclaration(None=@n4.B) {name=Skin, nodeColor=<rgba 1 0 0.0571 1>}
  NamedRerouteDeclaration(None=FunctionInput() {inputName=Material, inputType=FunctionInput_MaterialAttributes}) {name=MA, nodeColor=<rgba 1 0 0.3047 1>}
  NamedRerouteDeclaration(None=FunctionInput(Preview=Fresnel() {baseReflectFraction=0, exponent=1}) {bUsePreviewValueAsDefault=true, inputName=Fresnel, inputType=FunctionInput_Scalar, sortPriority=9}) {name=Fresnel, nodeColor=<rgba 0 0.6719 1 1>}
  n5 = LocalPosition() {localOrigin=InstancePreSkinning}
  n6 = VertexInterpolator(VS=@n5.XYZ)
  NamedRerouteDeclaration(None=FunctionInput(Preview=ShaderStageSwitch(PixelShader=@n6.PS, VertexShader=@n5.XYZ)) {bUsePreviewValueAsDefault=true, inputName=Pre Skinned Local Position, sortPriority=2}) {name=PSLP, nodeColor=<rgba 0 1 0.3209 1>}
  NamedRerouteDeclaration(None=FunctionInput(Preview=MaterialFunctionCall!(Material) {materialFunction=/Game/Packages/Fortress_SharedMaterials/Base_Material_Functions/CharacterBrightnessandScale_Balance}.Distance Blend) {bUsePreviewValueAsDefault=true, inputName=Distance Blend, inputType=FunctionInput_Scalar, sortPriority=3}) {name=DistanceBlend, nodeColor=<rgba 0 0.0806 1 1>}
 ]
}
