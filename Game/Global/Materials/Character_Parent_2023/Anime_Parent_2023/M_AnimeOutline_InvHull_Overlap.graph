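# Anime character parent material that draws inverted-hull outlines, with an alternate color where outlines overlap marked areas.
# World position offset comes from selectable normals and thickness driven by maps, panned noise and FOV-compensated distance fade, with optional camera-direction push.
# Opacity combines two-sided sign, thickness cutout and optional baked-normal fresnel.
# Overlap detection matches mask values against a GBuffer roughness or metallic band with a depth gate, bypassed on forward and mobile.
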Material /Game/Global/Materials/Character_Parent_2023/Anime_Parent_2023/M_AnimeOutline_InvHull_Overlap
graph {
 expressions {
  MaterialExpressionAbs_1 {class=MaterialExpressionAbs, inputs [None {from=MaterialExpressionDotProduct_0, output=""}], properties {}}
  MaterialExpressionAbs_2 {class=MaterialExpressionAbs, inputs [None {from=MaterialExpressionSubtract_3, output=""}], properties {}}
  MaterialExpressionAdd_0 {
   class=MaterialExpressionAdd
   inputs [
    A {from=MaterialExpressionLinearInterpolate_2}
    B {from=MaterialExpressionReroute_5}
   ]
  }
  MaterialExpressionAdd_2 {
   class=MaterialExpressionAdd
   inputs [
    A {from=MaterialExpressionNamedRerouteUsage_17}
    B {from=MaterialExpressionStaticSwitchParameter_12}
   ]
  }
  MaterialExpressionAdd_3 {
   class=MaterialExpressionAdd
   inputs [
    A {from=MaterialExpressionChannelMaskParameter_1, output=R}
    B {from=MaterialExpressionConstant_0}
   ]
  }
  MaterialExpressionAdd_4 {
   class=MaterialExpressionAdd
   inputs [
    A {from=MaterialExpressionMultiply_4}
    B {from=MaterialExpressionScalarParameter_6}
   ]
  }
  MaterialExpressionCameraPositionWS_0 {class=MaterialExpressionCameraPositionWS, inputs [], properties {}}
  MaterialExpressionCameraVectorWS_0 {class=MaterialExpressionCameraVectorWS, inputs [], properties {}}
  MaterialExpressionCeil_0 {
   class=MaterialExpressionCeil
   inputs [None {from=MaterialExpressionNamedRerouteDeclaration_5, output=""}]
   properties {desc=cut outlines mask}
  }
  MaterialExpressionChannelMaskParameter_0 {
   class=MaterialExpressionChannelMaskParameter
   inputs [None {from=MaterialExpressionStaticSwitchParameter_10, output=""}]
   properties {group=02_Outline Overlap Settings, parameterName=Outline Overlap Channel, sortPriority=4}
  }
  MaterialExpressionChannelMaskParameter_1 {
   class=MaterialExpressionChannelMaskParameter
   inputs [None {from=MaterialExpressionNamedRerouteDeclaration_14, output=""}]
   properties {group=01_InvertedHull_Outlines, maskChannel=Blue, parameterName=OutlineThickness Channel, sortPriority=3, defaultValue=<rgba 0 0 1 0>}
  }
  MaterialExpressionComponentMask_0 {class=MaterialExpressionComponentMask, inputs [None {from=MaterialExpressionTextureCoordinate_2}], properties {r=true}}
  MaterialExpressionComposite_0 {class=MaterialExpressionComposite, inputs [], properties {subgraphName=Construct Base Inputs}}
  MaterialExpressionConstant_0 {class=MaterialExpressionConstant, properties {desc=one bit used for opacity cutout, r=-0.009}}
  MaterialExpressionConstant_1 {class=MaterialExpressionConstant}
  MaterialExpressionConstant_2 {class=MaterialExpressionConstant, properties {r=1}}
  MaterialExpressionConstant_6 {class=MaterialExpressionConstant, properties {r=1}}
  MaterialExpressionDivide_0 {
   class=MaterialExpressionDivide
   inputs [
    A {from=MaterialExpressionVectorParameter_0}
    B {from=""}
   ]
   properties {constB=10}
  }
  MaterialExpressionDotProduct_0 {
   class=MaterialExpressionDotProduct
   inputs [
    A {from=MaterialExpressionCameraVectorWS_0, output=""}
    B {from=MaterialExpressionTransform_2, output=""}
   ]
   properties {}
  }
  MaterialExpressionFresnel_0 {
   class=MaterialExpressionFresnel
   inputs [
    ExponentIn {from="", output=""}
    BaseReflectFractionIn {from="", output=""}
    Normal {from="", output=""}
   ]
   properties {baseReflectFraction=0, exponent=1}
  }
  MaterialExpressionLinearInterpolate_0 {
   class=MaterialExpressionLinearInterpolate
   inputs [
    A {from=MaterialExpressionOneMinus_4}
    B {from=MaterialExpressionReroute_7}
    Alpha {from=MaterialExpressionScalarParameter_11}
   ]
  }
  MaterialExpressionLinearInterpolate_1 {
   class=MaterialExpressionLinearInterpolate
   inputs [
    A {from=MaterialExpressionReroute_0}
    B {from=MaterialExpressionMin_0}
    Alpha {from=MaterialExpressionPower_0}
   ]
  }
  MaterialExpressionLinearInterpolate_2 {
   class=MaterialExpressionLinearInterpolate
   inputs [
    A {from=""}
    B {from=MaterialExpressionMultiply_0}
    Alpha {from=MaterialExpressionStaticSwitchParameter_14}
   ]
   properties {constB=0}
  }
  MaterialExpressionLinearInterpolate_25 {
   class=MaterialExpressionLinearInterpolate
   inputs [
    A {from=MaterialExpressionNamedRerouteUsage_42}
    B {from=MaterialExpressionVectorParameter_6, output=RGB}
    Alpha {from=MaterialExpressionMultiply_9}
   ]
  }
  MaterialExpressionLinearInterpolate_3 {
   class=MaterialExpressionLinearInterpolate
   inputs [
    A {from=MaterialExpressionMultiply_5}
    B {from=MaterialExpressionNamedRerouteUsage_18}
    Alpha {from=MaterialExpressionScalarParameter_7}
   ]
  }
  MaterialExpressionLinearInterpolate_4 {
   class=MaterialExpressionLinearInterpolate
   inputs [
    A {from=MaterialExpressionLinearInterpolate_0}
    B {from=MaterialExpressionScalarParameter_23}
    Alpha {from=MaterialExpressionLinearInterpolate_0}
   ]
  }
  MaterialExpressionLinearInterpolate_8 {
   class=MaterialExpressionLinearInterpolate
   inputs [
    A {from=MaterialExpressionReroute_4}
    B {from=MaterialExpressionScalarParameter_40}
    Alpha {from=MaterialExpressionNamedRerouteUsage_0}
   ]
  }
  MaterialExpressionMaterialFunctionCall_1 {
   class=MaterialExpressionMaterialFunctionCall
   inputs [
    Material {from=MaterialExpressionNamedRerouteUsage_19}
    FX {from=MaterialExpressionNamedRerouteUsage_21}
   ]
   properties {materialFunction=/Game/Packages/Fortress_SharedMaterials/Anime_Master_Material/MFC_ClothWind}
  }
  MaterialExpressionMaterialFunctionCall_13 {
   class=MaterialExpressionMaterialFunctionCall
   inputs [
    In {from=MaterialExpressionMaterialFunctionCall_3, output=Result}
    LUT {from=""}
    LUT Multiplier {from=""}
   ]
   properties {materialFunction=/Game/Packages/Fortress_SharedMaterials/Anime_Master_Material/MF_ApplyLUT}
  }
  MaterialExpressionMaterialFunctionCall_14 {
   class=MaterialExpressionMaterialFunctionCall
   inputs [Material {from=""}]
   properties {materialFunction=/Game/Packages/Fortress_SharedMaterials/Base_Material_Functions/CharacterBrightnessandScale_Balance}
  }
  MaterialExpressionMaterialFunctionCall_16 {
   class=MaterialExpressionMaterialFunctionCall
   inputs [
    Material {from=MaterialExpressionMaterialFunctionCall_1, output=Result}
    M {from=MaterialExpressionReroute_8, output=M}
    Pre Skinned Local Position {from=MaterialExpressionReroute_9, output=Pre Skinned Local Position}
    Distance Blend {from=MaterialExpressionReroute_10, output=Distance Blend}
    Fresnel {from=MaterialExpressionReroute_11, output=Fresnel}
   ]
   properties {materialFunction=&MF_CharacterParent_2023_PostFX}
  }
  MaterialExpressionMaterialFunctionCall_2 {
   class=MaterialExpressionMaterialFunctionCall
   inputs [WPO {from=MaterialExpressionReroute_3}]
   properties {materialFunction=/Game/Packages/Fortress_SharedMaterials/Base_Material_Functions/Fortnite_HideElementsWithVertexColor_2022}
  }
  MaterialExpressionMaterialFunctionCall_25 {
   class=MaterialExpressionMaterialFunctionCall
   inputs [Color {from=MaterialExpressionMaterialFunctionCall_13, output=Result}]
   properties {materialFunction=/Game/Packages/Fortress_SharedMaterials/Anime_Master_Material/MF_TODColorAdjustment}
  }
  MaterialExpressionMaterialFunctionCall_3 {
   class=MaterialExpressionMaterialFunctionCall
   inputs [In {from=MaterialExpressionNamedRerouteUsage_3}]
   properties {materialFunction=/Game/Packages/Fortress_SharedMaterials/Anime_Master_Material/MF_TODCustomColors}
  }
  MaterialExpressionMaterialFunctionCall_4 {
   class=MaterialExpressionMaterialFunctionCall
   inputs [
    Fade Offset {from=MaterialExpressionScalarParameter_39}
    Fade Distance {from=MaterialExpressionScalarParameter_38}
   ]
   properties {materialFunction=/Game/Global/Materials/Character_Parent_2023/Anime_Parent_2023/MaterialFunctions/MFC_DistanceFade_FOV_compensation}
  }
  MaterialExpressionMaterialFunctionCall_5 {
   class=MaterialExpressionMaterialFunctionCall
   inputs [
    Normal {from=MaterialExpressionTextureSampleParameter2D_4, output=RGB}
    Flatness {from=MaterialExpressionScalarParameter_0}
   ]
   properties {materialFunction=&FlattenNormal}
  }
  MaterialExpressionMaterialFunctionCall_6 {
   class=MaterialExpressionMaterialFunctionCall
   inputs []
   properties {materialFunction=/Game/Global/Materials/Character_Parent_2023/Anime_Parent_2023/MaterialFunctions/MFC_ActorAlighnedScreenPosition}
  }
  MaterialExpressionMin_0 {
   class=MaterialExpressionMin
   inputs [
    A {from=MaterialExpressionReroute_0, output=""}
    B {from=MaterialExpressionNamedRerouteUsage_6, output=""}
   ]
   properties {}
  }
  MaterialExpressionMultiply_0 {
   class=MaterialExpressionMultiply
   inputs [
    A {from=MaterialExpressionNormalize_2}
    B {from=MaterialExpressionScalarParameter_2}
   ]
  }
  MaterialExpressionMultiply_1 {
   class=MaterialExpressionMultiply
   inputs [
    A {from=MaterialExpressionStaticSwitchParameter_4}
    B {from=MaterialExpressionNamedRerouteUsage_5}
   ]
  }
  MaterialExpressionMultiply_10 {
   class=MaterialExpressionMultiply
   inputs [
    A {from=MaterialExpressionVertexColor_0}
    B {from=""}
   ]
   properties {constB=2}
  }
  MaterialExpressionMultiply_11 {
   class=MaterialExpressionMultiply
   inputs [
    A {from=MaterialExpressionMultiply_14}
    B {from=MaterialExpressionNamedRerouteUsage_34}
   ]
  }
  MaterialExpressionMultiply_14 {
   class=MaterialExpressionMultiply
   inputs [
    A {from=MaterialExpressionOneMinus_8}
    B {from=MaterialExpressionStep_1}
   ]
  }
  MaterialExpressionMultiply_2 {
   class=MaterialExpressionMultiply
   inputs [
    A {from=MaterialExpressionCeil_0}
    B {from=MaterialExpressionOneMinus_1}
   ]
  }
  MaterialExpressionMultiply_21 {
   class=MaterialExpressionMultiply
   inputs [
    A {from=MaterialExpressionStaticSwitchParameter_6}
    B {from=MaterialExpressionNamedRerouteUsage_1}
   ]
  }
  MaterialExpressionMultiply_22 {
   class=MaterialExpressionMultiply
   inputs [
    A {from=MaterialExpressionTextureSampleParameter2D_0, output=RGB}
    B {from=MaterialExpressionScalarParameter_34}
   ]
  }
  MaterialExpressionMultiply_3 {
   class=MaterialExpressionMultiply
   inputs [
    A {from=MaterialExpressionNamedRerouteUsage_16}
    B {from=MaterialExpressionScalarParameter_4}
   ]
   properties {constB=0.1}
  }
  MaterialExpressionMultiply_4 {
   class=MaterialExpressionMultiply
   inputs [
    A {from=MaterialExpressionSubtract_1}
    B {from=MaterialExpressionScalarParameter_5}
   ]
  }
  MaterialExpressionMultiply_5 {
   class=MaterialExpressionMultiply
   inputs [
    A {from=MaterialExpressionStep_0}
    B {from=MaterialExpressionNamedRerouteUsage_18}
   ]
  }
  MaterialExpressionMultiply_6 {
   class=MaterialExpressionMultiply
   inputs [
    A {from=MaterialExpressionNamedRerouteUsage_7}
    B {from=MaterialExpressionStaticSwitchParameter_9}
   ]
  }
  MaterialExpressionMultiply_8 {
   class=MaterialExpressionMultiply
   inputs [
    A {from=MaterialExpressionOneMinus_0}
    B {from=MaterialExpressionStep_2}
   ]
  }
  MaterialExpressionMultiply_9 {
   class=MaterialExpressionMultiply
   inputs [
    A {from=MaterialExpressionMultiply_8}
    B {from=MaterialExpressionMultiply_11}
   ]
  }
  MaterialExpressionNamedRerouteDeclaration_0 {
   class=MaterialExpressionNamedRerouteDeclaration
   inputs [None {from=MaterialExpressionMaterialFunctionCall_4, output=Result}]
   properties {name=DistanceFading, nodeColor=<rgba 1 0.3047 0 1>}
  }
  MaterialExpressionNamedRerouteDeclaration_1 {
   class=MaterialExpressionNamedRerouteDeclaration
   inputs [None {from=MaterialExpressionPower_1}]
   properties {name=BakedNormalFresnel, nodeColor=<rgba 0 0.8144 1 1>}
  }
  MaterialExpressionNamedRerouteDeclaration_10 {
   class=MaterialExpressionNamedRerouteDeclaration
   inputs [None {from=MaterialExpressionStaticSwitchParameter_7}]
   properties {name=UVs, nodeColor=<rgba 1 0.6692 0 1>}
  }
  MaterialExpressionNamedRerouteDeclaration_11 {
   class=MaterialExpressionNamedRerouteDeclaration
   inputs [None {from=MaterialExpressionAdd_4}]
   properties {name=ThicknessNoise, nodeColor=<rgba 1 0.2972 0 1>}
  }
  MaterialExpressionNamedRerouteDeclaration_12 {
   class=MaterialExpressionNamedRerouteDeclaration
   inputs [None {from=MaterialExpressionStaticSwitchParameter_2}]
   properties {name=Outline_Material, nodeColor=<rgba 1 0 0.7734 1>}
  }
  MaterialExpressionNamedRerouteDeclaration_13 {
   class=MaterialExpressionNamedRerouteDeclaration
   inputs [None {from=MaterialExpressionVertexColor_1, output=A}]
   properties {name=VertexColorAlpha, nodeColor=<rgba 0.8984 1 0 1>}
  }
  MaterialExpressionNamedRerouteDeclaration_14 {
   class=MaterialExpressionNamedRerouteDeclaration
   inputs [None {from=MaterialExpressionTextureSampleParameter2D_3, output=RGB}]
   properties {name=FX_Map, nodeColor=<rgba 0 1 0.9063 1>}
  }
  MaterialExpressionNamedRerouteDeclaration_15 {
   class=MaterialExpressionNamedRerouteDeclaration
   inputs [None {from=MaterialExpressionOneMinus_9}]
   properties {name=DepthGate, desc=Depth gate to stop detection from background Gbuffer data, nodeColor=<rgba 1 0.1641 0 1>}
  }
  MaterialExpressionNamedRerouteDeclaration_18 {
   class=MaterialExpressionNamedRerouteDeclaration
   inputs [None {from=MaterialExpressionStaticSwitchParameter_11}]
   properties {name=OutlineColor, nodeColor=<rgba 0 1 0.1328 1>}
  }
  MaterialExpressionNamedRerouteDeclaration_2 {
   class=MaterialExpressionNamedRerouteDeclaration
   inputs [None {from=MaterialExpressionStaticSwitchParameter_8}]
   properties {name=WPO, nodeColor=<rgba 1 0.3281 0 1>}
  }
  MaterialExpressionNamedRerouteDeclaration_20 {
   class=MaterialExpressionNamedRerouteDeclaration
   inputs [None {from=MaterialExpressionMaterialFunctionCall_6, output=UVs}]
   properties {name=ObjectAlighedScreenPosition, nodeColor=<rgba 1 0.867 0 1>}
  }
  MaterialExpressionNamedRerouteDeclaration_3 {
   class=MaterialExpressionNamedRerouteDeclaration
   inputs [None {from=MaterialExpressionMultiply_1}]
   properties {name=OpacityMask, nodeColor=<rgba 0 0.8125 1 1>}
  }
  MaterialExpressionNamedRerouteDeclaration_4 {
   class=MaterialExpressionNamedRerouteDeclaration
   inputs [None {from=MaterialExpressionNormalize_0}]
   properties {name=BakedNormals, nodeColor=<rgba 0 0.2789 1 1>}
  }
  MaterialExpressionNamedRerouteDeclaration_5 {
   class=MaterialExpressionNamedRerouteDeclaration
   inputs [None {from=MaterialExpressionAdd_3}]
   properties {name=ThicknessMask, nodeColor=<rgba 1 0.1689 0 1>}
  }
  MaterialExpressionNamedRerouteDeclaration_6 {
   class=MaterialExpressionNamedRerouteDeclaration
   inputs [None {from=MaterialExpressionStaticSwitchParameter_0}]
   properties {name=Normal, nodeColor=<rgba 0.224 0.224 1 1>}
  }
  MaterialExpressionNamedRerouteDeclaration_7 {
   class=MaterialExpressionNamedRerouteDeclaration
   inputs [None {from=MaterialExpressionShadingPathSwitch_1}]
   properties {name=ColorRaw, nodeColor=<rgba 0 0.5781 1 1>}
  }
  MaterialExpressionNamedRerouteDeclaration_8 {
   class=MaterialExpressionNamedRerouteDeclaration
   inputs [None {from=MaterialExpressionMaterialFunctionCall_25, output=Result}]
   properties {name=ColorFinal, nodeColor=<rgba 1 0.6797 0 1>}
  }
  MaterialExpressionNamedRerouteDeclaration_9 {
   class=MaterialExpressionNamedRerouteDeclaration
   inputs [None {from=MaterialExpressionStaticSwitchParameter_23}]
   properties {name=Cinematics OutlineThicknessTextureOverride, nodeColor=<rgba 1 0 0.5391 1>}
  }
  MaterialExpressionNamedRerouteUsage_0 {class=MaterialExpressionNamedRerouteUsage}
  MaterialExpressionNamedRerouteUsage_1 {class=MaterialExpressionNamedRerouteUsage}
  MaterialExpressionNamedRerouteUsage_10 {class=MaterialExpressionNamedRerouteUsage}
  MaterialExpressionNamedRerouteUsage_11 {class=MaterialExpressionNamedRerouteUsage}
  MaterialExpressionNamedRerouteUsage_12 {class=MaterialExpressionNamedRerouteUsage}
  MaterialExpressionNamedRerouteUsage_13 {class=MaterialExpressionNamedRerouteUsage}
  MaterialExpressionNamedRerouteUsage_14 {class=MaterialExpressionNamedRerouteUsage}
  MaterialExpressionNamedRerouteUsage_15 {class=MaterialExpressionNamedRerouteUsage}
  MaterialExpressionNamedRerouteUsage_16 {class=MaterialExpressionNamedRerouteUsage}
  MaterialExpressionNamedRerouteUsage_17 {class=MaterialExpressionNamedRerouteUsage}
  MaterialExpressionNamedRerouteUsage_18 {class=MaterialExpressionNamedRerouteUsage}
  MaterialExpressionNamedRerouteUsage_19 {class=MaterialExpressionNamedRerouteUsage}
  MaterialExpressionNamedRerouteUsage_2 {class=MaterialExpressionNamedRerouteUsage}
  MaterialExpressionNamedRerouteUsage_20 {class=MaterialExpressionNamedRerouteUsage}
  MaterialExpressionNamedRerouteUsage_21 {class=MaterialExpressionNamedRerouteUsage}
  MaterialExpressionNamedRerouteUsage_3 {class=MaterialExpressionNamedRerouteUsage}
  MaterialExpressionNamedRerouteUsage_33 {class=MaterialExpressionNamedRerouteUsage}
  MaterialExpressionNamedRerouteUsage_34 {class=MaterialExpressionNamedRerouteUsage}
  MaterialExpressionNamedRerouteUsage_4 {class=MaterialExpressionNamedRerouteUsage}
  MaterialExpressionNamedRerouteUsage_42 {class=MaterialExpressionNamedRerouteUsage}
  MaterialExpressionNamedRerouteUsage_5 {class=MaterialExpressionNamedRerouteUsage}
  MaterialExpressionNamedRerouteUsage_6 {class=MaterialExpressionNamedRerouteUsage}
  MaterialExpressionNamedRerouteUsage_7 {class=MaterialExpressionNamedRerouteUsage}
  MaterialExpressionNamedRerouteUsage_8 {class=MaterialExpressionNamedRerouteUsage}
  MaterialExpressionNamedRerouteUsage_9 {class=MaterialExpressionNamedRerouteUsage}
  MaterialExpressionNormalize_0 {class=MaterialExpressionNormalize, inputs [VectorInput {from=MaterialExpressionSubtract_2, output=""}]}
  MaterialExpressionNormalize_1 {class=MaterialExpressionNormalize, inputs [VectorInput {from=MaterialExpressionMaterialFunctionCall_5, output=Result}]}
  MaterialExpressionNormalize_2 {class=MaterialExpressionNormalize, inputs [VectorInput {from=MaterialExpressionSubtract_0, output=""}]}
  MaterialExpressionOneMinus_0 {class=MaterialExpressionOneMinus, inputs [None {from=MaterialExpressionStep_10}]}
  MaterialExpressionOneMinus_1 {class=MaterialExpressionOneMinus, inputs [None {from=MaterialExpressionTwoSidedSign_0}]}
  MaterialExpressionOneMinus_2 {class=MaterialExpressionOneMinus, inputs [None {from=MaterialExpressionAbs_1}]}
  MaterialExpressionOneMinus_3 {class=MaterialExpressionOneMinus, inputs [None {from=MaterialExpressionTextureSampleParameter2D_1, output=R}]}
  MaterialExpressionOneMinus_4 {class=MaterialExpressionOneMinus, inputs [None {from=MaterialExpressionChannelMaskParameter_0, output=R}]}
  MaterialExpressionOneMinus_8 {class=MaterialExpressionOneMinus, inputs [None {from=MaterialExpressionStep_11}]}
  MaterialExpressionOneMinus_9 {class=MaterialExpressionOneMinus, inputs [None {from=MaterialExpressionStep_3}]}
  MaterialExpressionPanner_0 {
   class=MaterialExpressionPanner
   inputs [
    Coordinate {from=MaterialExpressionMultiply_3}
    Time {from=""}
    Speed {from=MaterialExpressionDivide_0}
   ]
  }
  MaterialExpressionPinBase_0 {class=MaterialExpressionPinBase, inputs [], properties {}}
  MaterialExpressionPinBase_1 {
   class=MaterialExpressionPinBase
   inputs [
    M {from=MaterialExpressionConstant_1, output=""}
    Pre Skinned Local Position {from=MaterialExpressionShaderStageSwitch_0, output=""}
    Distance Blend {from=MaterialExpressionMaterialFunctionCall_14, output=Distance Blend}
    Fresnel {from=MaterialExpressionFresnel_0, output=""}
   ]
   properties {reroutePins [
     M {expression=&self:MaterialGraph_1.MaterialGraphNode_Composite_0.Construct Base Inputs.MaterialExpressionReroute_8}
     Pre Skinned Local Position {expression=&self:MaterialGraph_1.MaterialGraphNode_Composite_0.Construct Base Inputs.MaterialExpressionReroute_9}
     Distance Blend {expression=&self:MaterialGraph_1.MaterialGraphNode_Composite_0.Construct Base Inputs.MaterialExpressionReroute_10}
     Fresnel {expression=&self:MaterialGraph_1.MaterialGraphNode_Composite_0.Construct Base Inputs.MaterialExpressionReroute_11}
    ]}
  }
  MaterialExpressionPixelDepth_0 {class=MaterialExpressionPixelDepth, inputs [], properties {}}
  MaterialExpressionPower_0 {
   class=MaterialExpressionPower
   inputs [
    Base {from=MaterialExpressionNamedRerouteDeclaration_13}
    Exp {from=MaterialExpressionScalarParameter_3}
   ]
  }
  MaterialExpressionPower_1 {
   class=MaterialExpressionPower
   inputs [
    Base {from=MaterialExpressionOneMinus_2}
    Exp {from=MaterialExpressionScalarParameter_1}
   ]
  }
  MaterialExpressionPreSkinnedPosition_0 {class=MaterialExpressionPreSkinnedPosition, inputs [], properties {}}
  MaterialExpressionReroute_0 {class=MaterialExpressionReroute, inputs [None {from=MaterialExpressionStaticSwitchParameter_3}]}
  MaterialExpressionReroute_1 {class=MaterialExpressionReroute, inputs [None {from=MaterialExpressionOneMinus_1}]}
  MaterialExpressionReroute_3 {class=MaterialExpressionReroute, inputs [None {from=MaterialExpressionMultiply_21}]}
  MaterialExpressionReroute_4 {class=MaterialExpressionReroute, inputs [None {from=MaterialExpressionScalarParameter_36}]}
  MaterialExpressionReroute_5 {class=MaterialExpressionReroute, inputs [None {from=MaterialExpressionStaticSwitchParameter_1}]}
  MaterialExpressionReroute_7 {class=MaterialExpressionReroute, inputs [None {from=MaterialExpressionChannelMaskParameter_0, output=R}]}
  MaterialExpressionScalarParameter_0 {
   class=MaterialExpressionScalarParameter
   properties {
    defaultValue=-0.1
    desc="values above 1.0 and below 0.0 can help with seperated geometry.\r\n-0.1 seems like a good wellyou to alight split geo."
    group=01_InvertedHull_Outlines
    parameterName=NormalMap_Flatness
    sliderMax=2
    sliderMin=-1
    sortPriority=5
   }
  }
  MaterialExpressionScalarParameter_1 {
   class=MaterialExpressionScalarParameter
   properties {defaultValue=5, group=01_InvertedHull_Outlines, parameterName=BakedNormalOpacity_Exponent, sliderMax=12}
  }
  MaterialExpressionScalarParameter_11 {
   class=MaterialExpressionScalarParameter
   properties {
    controlType=Enumeration
    enumeration=/Game/Global/Materials/Enums/E_Mat_YesNo
    group=02_Outline Overlap Settings
    parameterName=Invert Overlap Map
    sortPriority=5
   }
  }
  MaterialExpressionScalarParameter_12 {
   class=MaterialExpressionScalarParameter
   properties {
    defaultValue=0.9
    desc=The max value that the map detects as an overlap within the band
    enumeration=&E_Mat_YesNo
    group=02_Outline Overlap Settings
    parameterName=Overlap Detection Lower Band
    sortPriority=8
   }
  }
  MaterialExpressionScalarParameter_14 {
   class=MaterialExpressionScalarParameter
   properties {
    defaultValue=0.92
    desc=The min value that the map detects as an overlap within the band
    enumeration=&E_Mat_YesNo
    group=02_Outline Overlap Settings
    parameterName=Overlap Detection Upper Band
    sortPriority=6
   }
  }
  MaterialExpressionScalarParameter_16 {
   class=MaterialExpressionScalarParameter
   properties {
    defaultValue=0.92
    desc=The min value that the map detects as an overlap within the band
    enumeration=&E_Mat_YesNo
    group=02_Outline Overlap Settings
    parameterName=Overlap Detection Upper Band
    sortPriority=6
   }
  }
  MaterialExpressionScalarParameter_18 {
   class=MaterialExpressionScalarParameter
   properties {
    defaultValue=0.9
    desc=The max value that the map detects as an overlap within the band
    enumeration=&E_Mat_YesNo
    group=02_Outline Overlap Settings
    parameterName=Overlap Detection Lower Band
    sortPriority=8
   }
  }
  MaterialExpressionScalarParameter_2 {
   class=MaterialExpressionScalarParameter
   properties {defaultValue=0.3, group=01_InvertedHull_Outlines, parameterName=CameraDirectionOffset, sliderMax=2, sortPriority=3}
  }
  MaterialExpressionScalarParameter_23 {
   class=MaterialExpressionScalarParameter
   properties {
    defaultValue=0.91
    desc="The value which is detected by the overlap, must sit within the Upper and Lower overlap bands. NOTE this must match the Overlap Detection Value parameter set by the base material"
    group=02_Outline Overlap Settings
    parameterName=Overlap Detection Value
    sortPriority=7
   }
  }
  MaterialExpressionScalarParameter_3 {
   class=MaterialExpressionScalarParameter
   properties {
    defaultValue=1
    desc="Controls the exponent curve of the alpha channel to mask out the Baked NormalOpacity Fresnel effect with the Vertex Color alpha channel. \r\nto improve the masking if the alpha channel is used for opacity masking, because of smooth normal for the hair."
    group=01_InvertedHull_Outlines
    parameterName=BakedNormalOpacity_AlphaExponent
    sliderMax=8
   }
  }
  MaterialExpressionScalarParameter_34 {
   class=MaterialExpressionScalarParameter
   properties {
    defaultValue=1
    desc=Multiplier for the Outline Color Texture
    group=01_InvertedHull_Outlines
    parameterName=Color_Intensity
    sliderMax=1
    sortPriority=3
   }
  }
  MaterialExpressionScalarParameter_36 {
   class=MaterialExpressionScalarParameter
   properties {defaultValue=0.1, desc="LineThickness (Near)", group=01_InvertedHull_Outlines, parameterName=LineThickness, sliderMax=0.5, sortPriority=3}
  }
  MaterialExpressionScalarParameter_38 {
   class=MaterialExpressionScalarParameter
   properties {defaultValue=1000, group=01_InvertedHull_Outlines, parameterName=Fade_Distance, sliderMax=1500, sliderMin=500, sortPriority=5}
  }
  MaterialExpressionScalarParameter_39 {
   class=MaterialExpressionScalarParameter
   properties {defaultValue=50, group=01_InvertedHull_Outlines, parameterName=Fade_Offset, sliderMax=100, sortPriority=5}
  }
  MaterialExpressionScalarParameter_4 {
   class=MaterialExpressionScalarParameter
   properties {defaultValue=0.2, group=01_InvertedHull_Outlines, parameterName=ThicknessVariation_Tiling}
  }
  MaterialExpressionScalarParameter_40 {
   class=MaterialExpressionScalarParameter
   properties {defaultValue=0.6, group=01_InvertedHull_Outlines, parameterName=LineThickness_Far, sliderMax=2, sortPriority=3}
  }
  MaterialExpressionScalarParameter_5 {
   class=MaterialExpressionScalarParameter
   properties {defaultValue=1, group=01_InvertedHull_Outlines, parameterName=ThicknessVariation_Intensity, sliderMax=2}
  }
  MaterialExpressionScalarParameter_6 {
   class=MaterialExpressionScalarParameter
   properties {group=01_InvertedHull_Outlines, parameterName=ThicknessVariation_Bias, sliderMax=1, sliderMin=-1}
  }
  MaterialExpressionScalarParameter_7 {class=MaterialExpressionScalarParameter, properties {parameterName=IsToon?}}
  MaterialExpressionScalarParameter_8 {
   class=MaterialExpressionScalarParameter
   properties {
    defaultValue=100
    desc=Depth gate to stop detection from background Gbuffer data
    group=02_Outline Overlap Settings
    parameterName=Overlap Detection Depth Threshold
    sortPriority=9
   }
  }
  MaterialExpressionSceneDepth_5 {class=MaterialExpressionSceneDepth, inputs [UVs {from="", output=""}], properties {}}
  MaterialExpressionSceneTexture_18 {class=MaterialExpressionSceneTexture, inputs [UVs {from="", output=""}], properties {sceneTextureId=PPI_Roughness}}
  MaterialExpressionSceneTexture_8 {class=MaterialExpressionSceneTexture, inputs [UVs {from="", output=""}], properties {sceneTextureId=PPI_Metallic}}
  MaterialExpressionSetMaterialAttributes_1 {
   class=MaterialExpressionSetMaterialAttributes
   inputs [
    MaterialAttributes {from="", output=""}
    Emissive Color {from=MaterialExpressionNamedRerouteUsage_4, output=""}
    Opacity {from=MaterialExpressionLinearInterpolate_3, output=""}
    World Position Offset {from=MaterialExpressionNamedRerouteUsage_10, output=""}
   ]
   properties {attributeSetTypes [#1, #2, #3]}
  }
  MaterialExpressionSetMaterialAttributes_2 {
   class=MaterialExpressionSetMaterialAttributes
   inputs [
    MaterialAttributes {from="", output=""}
    Emissive Color {from=MaterialExpressionNamedRerouteUsage_2, output=""}
    Opacity Mask {from="", output=""}
    World Position Offset {from="", output=""}
   ]
   properties {
    desc=Debug
    attributeSetTypes [#1, #4, #3]
   }
  }
  MaterialExpressionShaderStageSwitch_0 {
   class=MaterialExpressionShaderStageSwitch
   inputs [
    PixelShader {from=MaterialExpressionVertexInterpolator_0, output=PS}
    VertexShader {from=MaterialExpressionPreSkinnedPosition_0, output=""}
   ]
   properties {}
  }
  MaterialExpressionShadingPathSwitch_1 {
   class=MaterialExpressionShadingPathSwitch
   inputs [
    Default {from=MaterialExpressionLinearInterpolate_25, output=""}
    Deferred {from="", output=""}
    Forward {from=MaterialExpressionNamedRerouteUsage_42, output=""}
    Mobile {from=MaterialExpressionNamedRerouteUsage_42, output=""}
   ]
   properties {}
  }
  MaterialExpressionStaticSwitchParameter_0 {
   class=MaterialExpressionStaticSwitchParameter
   inputs [
    True {from=MaterialExpressionTransform_0, output=""}
    False {from=MaterialExpressionStaticSwitchParameter_5, output=""}
   ]
   properties {desc=use normal map to prevent segmented geometry, group=01_InvertedHull_Outlines, parameterName=UseNormalMap, sortPriority=0}
  }
  MaterialExpressionStaticSwitchParameter_1 {
   class=MaterialExpressionStaticSwitchParameter
   inputs [
    True {from=MaterialExpressionMaterialFunctionCall_2, output=Result}
    False {from=MaterialExpressionReroute_3, output=""}
   ]
   properties {group=03_Opacity Mask, parameterName=Use Vertex Colors for Mask}
  }
  MaterialExpressionStaticSwitchParameter_10 {
   class=MaterialExpressionStaticSwitchParameter
   inputs [
    True {from=MaterialExpressionTextureSampleParameter2D_2, output=RGB}
    False {from=MaterialExpressionNamedRerouteUsage_9, output=""}
   ]
   properties {
    desc="This is needed when the character has customization where the overlap areas can change.\r\n\r\nUses channel from the Outline Map as defualt"
    group=02_Outline Overlap Settings
    parameterName=Use Separate Texture For Overlap Maps
    sortPriority=2
   }
  }
  MaterialExpressionStaticSwitchParameter_11 {
   class=MaterialExpressionStaticSwitchParameter
   inputs [
    True {from=MaterialExpressionMultiply_22, output=""}
    False {from=MaterialExpressionVectorParameter_5, output=RGB}
   ]
   properties {
    desc=Enable to use a texture to color the edge outlines of the mesh.
    group=01_InvertedHull_Outlines
    parameterName=UseDFL_ColorMap
    sortPriority=0
   }
  }
  MaterialExpressionStaticSwitchParameter_12 {
   class=MaterialExpressionStaticSwitchParameter
   inputs [
    True {from=MaterialExpressionLinearInterpolate_8, output=""}
    False {from=MaterialExpressionScalarParameter_36, output=""}
   ]
   properties {defaultValue=true, group=01_InvertedHull_Outlines, parameterName=UseDistanceFading, sortPriority=0}
  }
  MaterialExpressionStaticSwitchParameter_14 {
   class=MaterialExpressionStaticSwitchParameter
   inputs [
    True {from=MaterialExpressionNamedRerouteUsage_20, output=""}
    False {from=MaterialExpressionStaticSwitchParameter_15, output=""}
   ]
   properties {group=01_InvertedHull_Outlines, parameterName=UseOutlineOffsetVertexAlpha, sortPriority=0}
  }
  MaterialExpressionStaticSwitchParameter_15 {
   class=MaterialExpressionStaticSwitchParameter
   inputs [
    True {from=MaterialExpressionTextureSampleParameter2D_5, output=Mask}
    False {from=MaterialExpressionConstant_2, output=""}
   ]
   properties {group=01_InvertedHull_Outlines, parameterName=UseOutlineOffsetMap, sortPriority=0}
  }
  MaterialExpressionStaticSwitchParameter_19 {
   class=MaterialExpressionStaticSwitchParameter
   inputs [
    True {from=MaterialExpressionSceneTexture_8, output=Color}
    False {from=MaterialExpressionSceneTexture_18, output=Color}
   ]
   properties {
    desc="Use the Gbuffer Metallic channel for the outline overlap mask \r\n\r\nOverrides the the defualt Gbuffer Roughness Channel"
    group=02_Outline Overlap Settings
    parameterName="Use Gbuffer Metallic (Overrides Roughness Gbuffer)"
    sortPriority=10
   }
  }
  MaterialExpressionStaticSwitchParameter_2 {
   class=MaterialExpressionStaticSwitchParameter
   inputs [
    True {from=MaterialExpressionSetMaterialAttributes_2, output=""}
    False {from=MaterialExpressionSetMaterialAttributes_1, output=""}
   ]
   properties {group=01_InvertedHull_Outlines, parameterName=DebugThicknessMap, sortPriority=90}
  }
  MaterialExpressionStaticSwitchParameter_23 {
   class=MaterialExpressionStaticSwitchParameter
   inputs [
    True {from=MaterialExpressionOneMinus_3, output=""}
    False {from=MaterialExpressionConstant_6, output=""}
   ]
   properties {group=Cinematic, parameterName=UseCinematicsOutlineThicknessOverrideTexture}
  }
  MaterialExpressionStaticSwitchParameter_3 {
   class=MaterialExpressionStaticSwitchParameter
   inputs [
    True {from=MaterialExpressionMultiply_2, output=""}
    False {from=MaterialExpressionReroute_1, output=""}
   ]
   properties {desc=Influnces the global outline thickness, group=01_InvertedHull_Outlines, parameterName=UseThicknessMap, sortPriority=0}
  }
  MaterialExpressionStaticSwitchParameter_4 {
   class=MaterialExpressionStaticSwitchParameter
   inputs [
    True {from=MaterialExpressionLinearInterpolate_1, output=""}
    False {from=MaterialExpressionStaticSwitchParameter_3, output=""}
   ]
   properties {
    desc=mask out this effect with the vertec color alpha channel
    group=01_InvertedHull_Outlines
    parameterName=UseBakedNormalOpacityFresnel
    sortPriority=0
   }
  }
  MaterialExpressionStaticSwitchParameter_5 {
   class=MaterialExpressionStaticSwitchParameter
   inputs [
    True {from=MaterialExpressionTransform_1, output=""}
    False {from=MaterialExpressionVertexNormalWS_0, output=""}
   ]
   properties {
    desc=use baked normals in the vertex colors instead of the surface normals.
    group=01_InvertedHull_Outlines
    parameterName=UseBakedNormals
    sortPriority=0
   }
  }
  MaterialExpressionStaticSwitchParameter_6 {
   class=MaterialExpressionStaticSwitchParameter
   inputs [
    True {from=MaterialExpressionMultiply_6, output=""}
    False {from=MaterialExpressionStaticSwitchParameter_9, output=""}
   ]
   properties {desc=Influnces the global outline thickness, group=01_InvertedHull_Outlines, parameterName=UseThicknessMap, sortPriority=0}
  }
  MaterialExpressionStaticSwitchParameter_7 {
   class=MaterialExpressionStaticSwitchParameter
   inputs [
    True {from=MaterialExpressionTextureCoordinate_1, output=""}
    False {from=MaterialExpressionTextureCoordinate_0, output=""}
   ]
   properties {
    desc="active to use the second UV channel (UV1) for the Textures"
    group=01_InvertedHull_Outlines
    parameterName=UseSecondUVChannel
    sortPriority=0
   }
  }
  MaterialExpressionStaticSwitchParameter_8 {
   class=MaterialExpressionStaticSwitchParameter
   inputs [
    True {from=MaterialExpressionAdd_0, output=""}
    False {from=MaterialExpressionReroute_5, output=""}
   ]
   properties {group=01_InvertedHull_Outlines, parameterName=OutlineOffsetInCameraDirection, sortPriority=0}
  }
  MaterialExpressionStaticSwitchParameter_9 {
   class=MaterialExpressionStaticSwitchParameter
   inputs [
    True {from=MaterialExpressionAdd_2, output=""}
    False {from=MaterialExpressionStaticSwitchParameter_12, output=""}
   ]
   properties {group=01_InvertedHull_Outlines, parameterName=UseThicknessVariation, sortPriority=0}
  }
  MaterialExpressionStep_0 {
   class=MaterialExpressionStep
   inputs [
    Y {from=MaterialExpressionComponentMask_0, output=""}
    X {from="", output=""}
   ]
   properties {constY=1}
  }
  MaterialExpressionStep_1 {
   class=MaterialExpressionStep
   inputs [
    Y {from=MaterialExpressionScalarParameter_18, output=""}
    X {from=MaterialExpressionStaticSwitchParameter_19, output=""}
   ]
   properties {constX=0.1, constY=0.95}
  }
  MaterialExpressionStep_10 {
   class=MaterialExpressionStep
   inputs [
    Y {from=MaterialExpressionScalarParameter_14, output=""}
    X {from=MaterialExpressionLinearInterpolate_4, output=""}
   ]
   properties {constX=0.1, constY=0.95}
  }
  MaterialExpressionStep_11 {
   class=MaterialExpressionStep
   inputs [
    Y {from=MaterialExpressionScalarParameter_16, output=""}
    X {from=MaterialExpressionStaticSwitchParameter_19, output=""}
   ]
   properties {constX=0.1, constY=0.95}
  }
  MaterialExpressionStep_2 {
   class=MaterialExpressionStep
   inputs [
    Y {from=MaterialExpressionScalarParameter_12, output=""}
    X {from=MaterialExpressionLinearInterpolate_4, output=""}
   ]
   properties {constX=0.1, constY=0.95}
  }
  MaterialExpressionStep_3 {
   class=MaterialExpressionStep
   inputs [
    Y {from=MaterialExpressionScalarParameter_8, output=""}
    X {from=MaterialExpressionAbs_2, output=""}
   ]
   properties {}
  }
  MaterialExpressionSubtract_0 {
   class=MaterialExpressionSubtract
   inputs [
    A {from=MaterialExpressionCameraPositionWS_0}
    B {from=MaterialExpressionWorldPosition_0, output=XYZ}
   ]
  }
  MaterialExpressionSubtract_1 {
   class=MaterialExpressionSubtract
   inputs [
    A {from=MaterialExpressionTextureSample_0, output=R}
    B {from=""}
   ]
   properties {constB=0.5}
  }
  MaterialExpressionSubtract_2 {
   class=MaterialExpressionSubtract
   inputs [
    A {from=MaterialExpressionMultiply_10}
    B {from=""}
   ]
  }
  MaterialExpressionSubtract_3 {
   class=MaterialExpressionSubtract
   inputs [
    A {from=MaterialExpressionPixelDepth_0, output=R}
    B {from=MaterialExpressionSceneDepth_5, output=R}
   ]
  }
  MaterialExpressionTextureCoordinate_0 {class=MaterialExpressionTextureCoordinate}
  MaterialExpressionTextureCoordinate_1 {class=MaterialExpressionTextureCoordinate, properties {coordinateIndex=1}}
  MaterialExpressionTextureCoordinate_2 {class=MaterialExpressionTextureCoordinate}
  MaterialExpressionTextureSampleParameter2D_0 {
   class=MaterialExpressionTextureSampleParameter2D
   inputs [
    UVs {from=MaterialExpressionNamedRerouteUsage_14, output=""}
    Apply View MipBias {from="", output=""}
   ]
   properties {
    desc="Outline Color Texture, use can use the ShadedMap."
    group=01_InvertedHull_Outlines
    parameterName=DFL_Color_Map
    sortPriority=1
    texture=&T_BlacksRGB
   }
  }
  MaterialExpressionTextureSampleParameter2D_1 {
   class=MaterialExpressionTextureSampleParameter2D
   inputs [
    UVs {from=MaterialExpressionNamedRerouteUsage_15, output=""}
    Apply View MipBias {from="", output=""}
   ]
   properties {
    group=Cinematic
    parameterName=CinematicsOutlineThicknessOverride
    samplerType=SAMPLERTYPE_Masks
    texture=/Game/Global/Textures/Default/Blanks/LargeEmpty
   }
  }
  MaterialExpressionTextureSampleParameter2D_2 {
   class=MaterialExpressionTextureSampleParameter2D
   inputs [
    UVs {from=MaterialExpressionNamedRerouteUsage_33, output=""}
    Apply View MipBias {from="", output=""}
   ]
   properties {
    desc="Separate Outline Overlap Map - This is needed when the character has customization where the overlap areas can change. Each channel can contain a separate overlap mask which is controlled in the CID custom options by setting the Outline Overlap Channel within the varient "
    group=02_Outline Overlap Settings
    parameterName=Outline Overlap Map
    samplerType=SAMPLERTYPE_Masks
    sortPriority=3
    texture=/Game/Packages/Fortress_SharedMaterials/Anime_Master_Material/T_Outline_Mask
    channelNames {a="", b=Overlap Map 03, g=Overlap Map 02, r=Overlap Map 01}
   }
  }
  MaterialExpressionTextureSampleParameter2D_3 {
   class=MaterialExpressionTextureSampleParameter2D
   inputs [
    UVs {from=MaterialExpressionNamedRerouteUsage_12, output=""}
    Apply View MipBias {from="", output=""}
   ]
   properties {
    desc=This can be used for the Outline Thickness Map and Outline Ovelap Map
    group=01_InvertedHull_Outlines
    parameterName=Outline Map
    samplerType=SAMPLERTYPE_Masks
    sortPriority=1
    texture=&T_Outline_Mask
    channelNames {a="", b=Unused, g=Outline Overlap Map, r=Outline Thickness Map}
   }
  }
  MaterialExpressionTextureSampleParameter2D_4 {
   class=MaterialExpressionTextureSampleParameter2D
   inputs [
    UVs {from=MaterialExpressionNamedRerouteUsage_13, output=""}
    Apply View MipBias {from="", output=""}
   ]
   properties {
    group=01_InvertedHull_Outlines
    parameterName=Normal_Map
    samplerType=SAMPLERTYPE_Normal
    sortPriority=1
    texture=&FlatNormalBlue
   }
  }
  MaterialExpressionTextureSampleParameter2D_5 {
   class=MaterialExpressionTextureSampleParameter2D
   inputs [
    UVs {from="", output=""}
    Apply View MipBias {from="", output=""}
   ]
   properties {
    group=01_InvertedHull_Outlines
    parameterName=OutlineOffset_Map
    sortPriority=1
    texture=&T_WhitesRGB
    channelNames {a="", b="", g="", r=Mask}
   }
  }
  MaterialExpressionTextureSample_0 {
   class=MaterialExpressionTextureSample
   inputs [
    UVs {from=MaterialExpressionPanner_0}
    Tex {from=""}
    Apply View MipBias {from=""}
   ]
   properties {samplerSource=SSM_Wrap_WorldGroupSettings, samplerType=SAMPLERTYPE_LinearColor, texture=/EpicBaseTextures/Noises/T_Perlin_Noise_D}
  }
  MaterialExpressionTransform_0 {class=MaterialExpressionTransform, inputs [None {from=MaterialExpressionNormalize_1}]}
  MaterialExpressionTransform_1 {
   class=MaterialExpressionTransform
   inputs [None {from=MaterialExpressionNamedRerouteDeclaration_4}]
   properties {transformSourceType=TRANSFORMSOURCE_Local}
  }
  MaterialExpressionTransform_2 {
   class=MaterialExpressionTransform
   inputs [None {from=MaterialExpressionNamedRerouteUsage_11}]
   properties {transformSourceType=TRANSFORMSOURCE_Local}
  }
  MaterialExpressionTwoSidedSign_0 {class=MaterialExpressionTwoSidedSign, inputs [], properties {}}
  MaterialExpressionVectorParameter_0 {
   class=MaterialExpressionVectorParameter
   properties {group=01_InvertedHull_Outlines, parameterName=ThicknessVariation_PannerSpeed, channelNames {a="", b="", g=Speed V, r=Speed U}}
  }
  MaterialExpressionVectorParameter_5 {
   class=MaterialExpressionVectorParameter
   properties {group=01_InvertedHull_Outlines, parameterName=Outline Color, sortPriority=2, defaultValue=<rgba 0.005 0.005 0.005 1>}
  }
  MaterialExpressionVectorParameter_6 {
   class=MaterialExpressionVectorParameter
   properties {group=02_Outline Overlap Settings, parameterName=Outline Overlap Color, sortPriority=1, defaultValue=<rgba 1 1 1 1>}
  }
  MaterialExpressionVertexColor_0 {class=MaterialExpressionVertexColor, inputs [], properties {}}
  MaterialExpressionVertexColor_1 {class=MaterialExpressionVertexColor, inputs [], properties {}}
  MaterialExpressionVertexInterpolator_0 {class=MaterialExpressionVertexInterpolator, inputs [VS {from=MaterialExpressionPreSkinnedPosition_0, output=""}], properties {}}
  MaterialExpressionVertexNormalWS_0 {class=MaterialExpressionVertexNormalWS, inputs [], properties {}}
  MaterialExpressionWorldPosition_0 {class=MaterialExpressionWorldPosition, inputs [], properties {}}
 }
 properties {
  bAutomaticallySetUsageInEditor=false
  bEnableResponsiveAA=true
  bUseMaterialAttributes=true
  bUsedWithClothing=true
  bUsedWithMeshDeformer=true
  bUsedWithMorphTargets=true
  bUsedWithSkeletalMesh=true
  blendMode=BLEND_Translucent
  pixelDepthOffsetMode=PDOM_Legacy
  shadingModel=MSM_Unlit
  twoSided=true
  usedShadingModels=Unlit
 }
}
