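# Anime character parent material on the Cloth shading model that flags outline overlap areas for a separate Outline Overlap material.
# Shared Pre, Calculations and Post functions build the toon shading, with masked eye animation applied to the colours before Post.
# An invertible mask becomes a detection value that replaces Roughness or Metallic by parameter, using the GBuffer channel as a marker.
# An IsKicks switch routes the result through the sock function.
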Material /Game/Global/Materials/Character_Parent_2023/Anime_Parent_2023/M_AnimeShader_OutlineOverlap
graph {
 properties {
  bAutomaticallySetUsageInEditor=false
  bUseMaterialAttributes=true
  bUsedWithClothing=true
  bUsedWithMeshDeformer=true
  bUsedWithMorphTargets=true
  bUsedWithNiagaraMeshParticles=true
  bUsedWithSkeletalMesh=true
  bUsedWithStaticLighting=true
  pixelDepthOffsetMode=PDOM_Legacy
  shadingModel=MSM_Cloth
  usedShadingModels=Cloth
 }
 nodes [
  n1 = TextureSampleParameter2D() {
   desc=This is a binary mask that is used to detect the areas where the outline is overlapping
   group=02_Outline Overlap Settings
   parameterName=Outline Overlap Map
   samplerSource=SSM_Wrap_WorldGroupSettings
   samplerType=SAMPLERTYPE_Masks
   sortPriority=1
   texture=&T_BlackMask
  }
  n2 = ChannelMaskParameter(None=@n1.RGB) {group=02_Outline Overlap Settings, parameterName=Outline Overlap Channel, sortPriority=2}
  n3 = ScalarParameter() {
   controlType=Enumeration
   enumeration=/Game/Global/Materials/Enums/E_Mat_YesNo
   group=02_Outline Overlap Settings
   parameterName=Invert Overlap Map
   sortPriority=4
  }
  n4 = Saturate(None=LinearInterpolate(A=OneMinus(None=@n2.R), B=@n2.R, Alpha=@n3))
  n5 = ScalarParameter() {
   defaultValue=0.91
   desc="The value which is detected by the overlap. NOTE this must match the Overlap Detection Value parameter and  fall within the lower and upper Overlap Detection Bands set within the Outline Overlap Material "
   group=02_Outline Overlap Settings
   parameterName=Overlap Detection Value
   sortPriority=3
  }
  NamedRerouteDeclaration(None=LinearInterpolate(A=@n4, B=@n5, Alpha=@n4)) {name=OverlapMap, nodeColor=<rgba 0.7812 1 0 1>}
  n6 = MaterialFunctionCall!(CustomUV) {materialFunction=/Game/Global/Materials/Character_Parent_2023/Anime_Parent_2023/MaterialFunctions/MFC_AnimeShader_2023_Pre}
  n7 = MaterialFunctionCall!(CustomLightVector, OverrideNormal_WS, STT=@n6.STT, Normal_Map=@n6.STT, DFL=@n6.STT, FX=@n6.STT) {materialFunction=/Game/Global/Materials/Character_Parent_2023/Anime_Parent_2023/MaterialFunctions/MFC_AnimeShader_2023_Calculations}
  n8 = VertexColor()
  n9 = TextureSampleParameter2D() {
   group=01_EyeAnimation_Parameters
   parameterName=EyeMask
   samplerSource=SSM_Wrap_WorldGroupSettings
   samplerType=SAMPLERTYPE_Masks
   sortPriority=-32
   texture=&T_WhiteMask
  }
  n10 = StaticSwitchParameter(True=AppendVector(A=@n8, B=@n8), False=@n9.RGBA) {group=01_EyeAnimation_Parameters, parameterName=UseVertexColorForEyeMask, sortPriority=-33}
  n11 = TextureCoordinate() {coordinateIndex=1}
  n12 = Sign(None=Subtract(A=ComponentMask(None=@n11) {r=true}) {constB=0.0001})
  n13 = MaterialFunctionCall!(
   Color Lit=@n6.Color_Lit
   Color Shaded=@n6.Color_Lit
   Use UV2=StaticBoolParameter() {defaultValue=true, group=01_EyeAnimation_Parameters, parameterName=Use UV2 for Eye Animation, sortPriority=-30}
   Mask=StaticSwitchParameter(
    True=ChannelMaskParameter(None=@n10) {group=01_EyeAnimation_Parameters, parameterName=EyeMaskChannel, sortPriority=-32}.R
    False=Saturate(None=@n12)
   ) {group=01_EyeAnimation_Parameters, parameterName=UseEyeMask, sortPriority=-33}
  ) {materialFunction=/Game/Global/Materials/Character_Parent_2023/Anime_Parent_2023/MaterialFunctions/MF_AnimeEyeAnimation_v2}
  n14 = GetMaterialAttributes(None=MaterialFunctionCall!(
    Shading_Gradient=@n7.Shading_Gradient
    Threshold_Gradient=@n7.Shading_Gradient
    Specular_Gradient=@n7.Shading_Gradient
    Sheen_Mask=@n7.Shading_Gradient
    Thickness=@n7.Shading_Gradient
    DFL_Mask=@n7.Shading_Gradient
    FX=@n7.Shading_Gradient
    Metallic_Gradient=@n7.Shading_Gradient
    RimLight_Gradient=@n7.Shading_Gradient
    NormalBuffer=@n7.Shading_Gradient
    --------
    CINE_LightSources=@n7.Shading_Gradient
    CINE_LightColor=@n7.Shading_Gradient
    ---------
    Color_Lit=@n13.Lit
    Color_Shaded=@n13.Lit
    SRM=@n6.SRM
    Emissive=@n6.SRM
    DFL_Color=@n6.SRM
    Specular_Color=@n6.SRM
    RimLight_Color=@n6.SRM
   ) {materialFunction=/Game/Global/Materials/Character_Parent_2023/Anime_Parent_2023/MaterialFunctions/MFC_AnimeShader_2023_Post}.Result) {attributeGetTypes [#1, #2]}
  n15 = ScalarParameter() {
   controlType=Enumeration
   defaultValue=1
   desc="Needs to match the Gbuffer Channel used by the Outline Overlap Material "
   enumeration=&E_Mat_YesNo
   group=02_Outline Overlap Settings
   parameterName=Apply Outline Overlap Map to Roughness
   sortPriority=5
  }
  n16 = ScalarParameter() {
   controlType=Enumeration
   desc="Needs to match the Gbuffer Channel used by the Outline Overlap Material "
   enumeration=&E_Mat_YesNo
   group=02_Outline Overlap Settings
   parameterName=Apply Outline Overlap Map to Metallic
   sortPriority=6
  }
  n17 = SetMaterialAttributes!(
   MaterialAttributes=@n14.MaterialAttributes
   Roughness=LinearInterpolate(A=@n14.Roughness, B=NamedRerouteUsage(), Alpha=@n15)
   Metallic=LinearInterpolate(A=@n14.Metallic, B=NamedRerouteUsage(), Alpha=@n16)
  ) {attributeSetTypes [#1, #2]}
  StaticSwitchParameter(
   True=MaterialFunctionCall!(InOpacity, Material=@n17, Mask) {materialFunction=/Game/Global/Materials/Cosmetics/Kicks/Socks/Materials/MF_Sock}.Result
   False=@n17
  ) {group=Kicks, parameterName=IsKicks}
  TemporalResponsivenessOutput!(None)
 ]
}
