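# Character parent material function that adds rim lighting to the emissive of incoming material attributes.
# It builds a Fresnel mask from AO and upward normals, a pre-skinned height gradient, close-to-distance brightness terms, and two directional rim lights tinted by time of day.
# Named reroutes carry these terms into a quality switch, with a separate Epic blend and an optional close-range path otherwise.
# A global collection parameter gates the result before boost multipliers apply.
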MaterialFunction /Game/Global/Materials/Character_Parent_2023/Functions/MF_Character_Parent_2023_Rim
graph {nodes [
  Desaturation(
   None=MaterialFunctionCall() {materialFunction=/Game/Effects/Fort_Effects/Materials/_Functions/CurveColor_Time_of_Day}.Result
   Fraction=Constant() {r=0.5}
  )
  n1 = NamedRerouteUsage()
  n2 = NamedRerouteUsage()
  n3 = Multiply(A=Saturate(None=Add(A=@n2) {constB=0.2}), B=@n1)
  n4 = NamedRerouteUsage()
  n5 = LinearInterpolate(A=NamedRerouteUsage(), B=@n4, Alpha=NamedRerouteUsage())
  n6 = ScalarParameter() {defaultValue=1, group=RimLight, parameterName=Advanced_Rim_ThirdParty}
  n7 = CollectionParameter() {
   collection=&FortniteMaterialParameters
   parameterName=bEnableCharacterRimLight
  }
  n8 = Multiply(
   A=QualitySwitch(
    Default=StaticSwitchParameter(True=@n3, False=@n1) {group=RimLight-Close, parameterName=Enable CloseRangeRimlight, sortPriority=0}
    Epic=LinearInterpolate(A=@n4, B=@n5, Alpha=@n6)
   )
   B=@n7
  )
  n9 = ScalarParameter() {defaultValue=2, group=RimLight, parameterName=RimlightEyeExposureBoost}
  n10 = Multiply(
   A=ScalarParameter() {group=RimLight, parameterName=RimLight_Multiplier}
   B=Multiply(A=ScalarParameter() {defaultValue=1, group=RimLight, parameterName=Rimlight_CvarBoost}, B=Multiply(A=@n8, B=@n9))
  )
  n11 = Constant3Vector() {constant=<rgba 0.8021 0.7085 0 0>}
  n12 = GetMaterialAttributes(None=FunctionInput(Preview=SetMaterialAttributes!(MaterialAttributes, EmissiveColor=@n11) {attributeSetTypes [#1]}) {inputName=Material, inputType=FunctionInput_MaterialAttributes}) {attributeGetTypes [#2, #3, #1]}
  FunctionOutput(None=SetMaterialAttributes(MaterialAttributes=@n12.MaterialAttributes, EmissiveColor=Add(A=@n12.EmissiveColor, B=@n10)) {attributeSetTypes [#1]})
  n13 = PreSkinnedPosition()
  n14 = Divide(
   A=FunctionInput(Preview=VertexInterpolator(VS=@n13).PS) {bUsePreviewValueAsDefault=true, inputName=Pre Skinned Local Position}
   B=Constant() {r=170}
  )
  NamedRerouteDeclaration(None=Saturate(None=ComponentMask(None=@n14) {b=true})) {name=Z-Gradient, nodeColor=<rgba 0 0.9062 1 1>}
  n15 = Constant() {r=2}
  n16 = Constant() {r=2}
  n17 = VertexNormalWS()
  n18 = Multiply(
   A=Fresnel(
    ExponentIn=StaticSwitchParameter(True=@n15, False=@n16) {group=RimLight-Close, parameterName=Enable CloseRangeRimlight, sortPriority=0}
    Normal=Transform(None=@n12.Normal)
   ) {baseReflectFraction=0}
   B=Multiply(
    A=FunctionInput() {inputName=AO, inputType=FunctionInput_Scalar, previewValue=<1 0 0 1>}
    B=MaterialFunctionCall(Float3=@n17) {materialFunction=&BreakOutFloat3Components}.B
   )
  )
  n19 = Constant() {r=0.03}
  NamedRerouteDeclaration(None=Saturate(None=MaterialFunctionCall(In=@n18, Contrast=@n19) {materialFunction=&CheapContrast}.Result)) {name=RimLightFresnel, nodeColor=<rgba 1 0.375 0 1>}
  n20 = Constant() {r=1}
  n21 = MaterialFunctionCall(
   Color_Morning=Constant3Vector() {constant=<rgba 0.875 0.8626 0.7474 1>}
   Color_Day=Constant3Vector() {constant=<rgba 1.2 0.9092 0.5938 1>}
   Color_Evening=Constant3Vector() {constant=<rgba 0.9635 0.5427 0.2158 1>}
   Color_Night=Constant3Vector() {constant=<rgba 0.3927 0.3462 0.651 1>}
  ) {materialFunction=&Color_Time_of_Day}
  n22 = StaticSwitchParameter(True=Multiply(A=@n20, B=@n21.Result), False=@n20) {group=RimLight-Close, parameterName=Enable TODCurveColorMult, sortPriority=0}
  n23 = Constant() {r=1}
  n24 = MaterialFunctionCall!(
   Blend Range=ScalarParameter() {defaultValue=800, desc=4500, group=RimLight, parameterName=RimLightCloseRangeBlendDistance}
   Start Offset=ScalarParameter() {defaultValue=-2000, group=RimLight, parameterName=RimDistanceStartOffset}
  ) {materialFunction=/Engine/Functions/Engine_MaterialFunctions03/Particles/Distance_Blend}
  NamedRerouteDeclaration(None=StaticSwitchParameter(True=LinearInterpolate(A=@n22, B=@n23, Alpha=@n24.Result), False=@n24.Result) {group=RimLight-Close, parameterName=Enable CloseRangeRimlight, sortPriority=0}) {name=CloseRangeEmissive, nodeColor=<rgba 0 0.0625 1 1>}
  n25 = MaterialFunctionCall(
   X=ScalarParameter() {defaultValue=2, group=RimLight-Close, parameterName=Close Range Rim Brightness}
   Y=ScalarParameter() {defaultValue=4, group=RimLight, parameterName=Distance Rim Brightness}
  ) {materialFunction=&MakeFloat2}
  n26 = MaterialFunctionCall(Float2=QualitySwitch(Default=@n25.Result, Epic=Multiply(A=@n25.Result) {constB=0.75})) {materialFunction=&BreakOutFloat2Components}
  n27 = CollectionParameter() {
   collection=&FortniteMaterialParameters
   parameterName=DistanceRimBrightnessMultiply
  }
  n28 = Constant()
  n29 = Multiply(
   A=Clamp(None=@n12.BaseColor) {maxDefault=0.75, minDefault=0.25}
   B=LinearInterpolate(A=@n26.R, B=Multiply(A=@n26.G, B=@n27), Alpha=FunctionInput(Preview=@n28) {inputName=Distance Blend, inputType=FunctionInput_Scalar})
  )
  n30 = NamedRerouteUsage()
  NamedRerouteDeclaration(None=Multiply(A=LinearInterpolate(B=@n29, Alpha=@n30), B=NamedRerouteUsage())) {name=BrightnessAdjustments, nodeColor=<rgba 0.6641 0 1 1>}
  n31 = NamedRerouteUsage()
  n32 = FunctionInput(Preview=TextureSample() {samplerType=SAMPLERTYPE_Grayscale, texture=/Game/ArtTools/MaterialLibrary/Materials/T_MasterKey_Swatch_GMap}.G) {bUsePreviewValueAsDefault=true, inputName=Gmap, inputType=FunctionInput_Scalar}
  n33 = Constant() {r=2}
  n34 = Saturate(None=Multiply(A=Power(Base=@n32, Exp=@n33), B=Constant() {r=1}))
  n35 = Desaturation(None=Clamp(None=@n12.BaseColor) {maxDefault=0.7, minDefault=0.03}, Fraction=Constant() {r=0.6})
  n36 = MaterialFunctionCall!(
   BaseColor=@n35
   Gmap=@n34
   LightVector=VectorParameter() {group=RimLight, parameterName=RimLightDirection01, defaultValue=<rgba 0.376 1.594 2 1>}.RGB
  ) {materialFunction=/Game/Packages/Fortress_SharedMaterials/Base_Material_Functions/MF_Rimlight_SingleLight}
  n37 = MaterialFunctionCall(
   BaseColor=@n35
   Gmap=@n34
   LightVector=VectorParameter() {group=RimLight, parameterName=RimLightDirection02, defaultValue=<rgba 0.9094 0.8438 1.5 1>}.RGB
  ) {materialFunction=&MF_Rimlight_SingleLight}
  n38 = Multiply(A=Multiply(A=@n21.Result, B=Add(A=@n36.Result, B=@n37.Result)), B=Constant() {desc="artemis = 2.2 - Asteria 6(pending)", r=6})
  NamedRerouteDeclaration(None=Clamp(None=Multiply(A=@n31, B=@n38)) {maxDefault=500}) {name=RimLightColors, nodeColor=<rgba 0 1 0.1094 1>}
  PreSkinnedLocalBounds()
 ]}
