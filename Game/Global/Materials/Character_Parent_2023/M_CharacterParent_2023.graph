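# Parent material for characters on skeletal meshes, supporting morph targets and mesh deformers.
# It uses material attributes and takes its shading model from the graph, covering Default Lit, Subsurface, Subsurface Profile, Clear Coat and Cloth.
# The logic lives in two material functions, with MF_CharacterParent_2023_PreFX feeding MF_CharacterParent_2023_PostFX.
# Named reroutes expose the material attributes and extracted emissive color as connection points within the graph.
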Material /Game/Global/Materials/Character_Parent_2023/M_CharacterParent_2023
graph {
 properties {
  bAutomaticallySetUsageInEditor=false
  bUseMaterialAttributes=true
  bUsedWithMeshDeformer=true
  bUsedWithMorphTargets=true
  bUsedWithSkeletalMesh=true
  pixelDepthOffsetMode=PDOM_Legacy
  shadingModel=MSM_FromMaterialExpression
  usedShadingModels=Default Lit | Subsurface | Clear Coat | Subsurface Profile | Cloth
 }
 nodes [
  n1 = MaterialFunctionCall() {materialFunction=&MF_CharacterParent_2023_PreFX}
  MaterialFunctionCall(Material=@n1.Result, M=@n1.Result, Pre Skinned Local Position=@n1.Result, Distance Blend=@n1.Result, Fresnel=@n1.Result) {materialFunction=&MF_CharacterParent_2023_PostFX}
  NamedRerouteDeclaration(None=@n1.Result) {name=Material In, nodeColor=<rgba 0 1 0.4375 1>}
  n2 = NamedRerouteUsage()
  NamedRerouteDeclaration(None=@n2) {name=Material Out, nodeColor=<rgba 1 0 0.9609 1>}
  NamedRerouteDeclaration(None=GetMaterialAttributes(None=@n2) {attributeGetTypes [#1]}.Emissive Color) {name=Emissive In, nodeColor=<rgba 0 1 0.7188 1>}
 ]
}
