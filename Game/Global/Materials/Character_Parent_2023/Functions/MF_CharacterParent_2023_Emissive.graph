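# Material function for the 2023 character parent material that optionally sets EmissiveColor on incoming material attributes.
# A tinted emissive texture, full or cropped through TextureCropping and optionally squared, feeds the color, with a static switch allowing tint alone.
# Static switches add invertible, power-shaped Fresnel and FX channel masking.
# A brightness parameter includes an extra mobile multiplier applied outside the deferred shading path.
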MaterialFunction /Game/Global/Materials/Character_Parent_2023/Functions/MF_CharacterParent_2023_Emissive
graph {nodes [
  n1 = FunctionInput() {inputName=Material, inputType=FunctionInput_MaterialAttributes}
  n2 = VectorParameter() {group=06_Emissive, parameterName=EmissiveTint, sortPriority=6, defaultValue=<rgba 1 1 1 1>}
  n3 = TextureSampleParameter2D(UVs=NamedRerouteUsage()) {
   group=06_Emissive
   parameterName=Emissive
   samplerSource=SSM_Wrap_WorldGroupSettings
   sortPriority=1
   texture=&T_BlacksRGB
  }
  n4 = StaticSwitchParameter(True=NamedRerouteUsage(), False=@n3.RGB) {desc=CROP_EMIS, group=06_Emissive, parameterName=UseCroppedEmissive}
  n5 = StaticSwitchParameter(True=Multiply(A=@n4, B=@n4), False=@n4) {group=06_Emissive, parameterName=UseLinearEmissiveMask}
  n6 = Multiply(A=StaticSwitch(True=@n2.RGB, False=Multiply(A=@n2.RGB, B=@n5), Value=NamedRerouteUsage()), B=NamedRerouteUsage())
  n7 = StaticSwitchParameter(True=Multiply(A=@n6, B=NamedRerouteUsage()), False=@n6) {group=06_Emissive, parameterName=UseEmissiveFresnel}
  n8 = NamedRerouteUsage()
  n9 = StaticSwitch(
   True=Multiply(A=@n7, B=ChannelMaskParameter(None=@n8) {group=06_Emissive, parameterName=EmissiveFXMaskChannel, sortPriority=101}.R)
   False=@n7
   Value=NamedRerouteDeclaration(None=StaticBoolParameter() {group=06_Emissive, parameterName=UseFXMaskForEmissive, sortPriority=100}) {name=UseFXMap, nodeColor=<rgba 0 1 0.0859 1>}
  )
  FunctionOutput(None=StaticSwitchParameter(True=SetMaterialAttributes!(MaterialAttributes=@n1, EmissiveColor=@n9) {attributeSetTypes [#1]}, False=@n1) {group=04_Tech Art, parameterName=UseBasicEmissive, sortPriority=0})
  n10 = ScalarParameter() {defaultValue=1, group=06_Emissive, parameterName=EmissiveMobileMultiplier}
  n11 = Constant() {r=1}
  NamedRerouteDeclaration(None=Multiply(A=ScalarParameter() {defaultValue=1, group=06_Emissive, parameterName=EmissiveBrightness}, B=ShadingPathSwitch(Default=@n10, Deferred=@n11))) {name=Emissive Brightness Modulation, nodeColor=<rgba 1 0.3281 0 1>}
  n12 = FunctionInput() {inputName=Fresnel, inputType=FunctionInput_Scalar, sortPriority=2}
  n13 = StaticSwitchParameter(True=@n12, False=OneMinus(None=@n12)) {group=06_Emissive, parameterName=InvertEmissiveFresnel}
  n14 = ScalarParameter() {defaultValue=4, group=06_Emissive, parameterName=EmissiveFresnelPower}
  NamedRerouteDeclaration(None=Saturate(None=Power(Base=@n13, Exp=@n14))) {name=Fresnel Processing, nodeColor=<rgba 0 0.7187 1 1>}
  n15 = TextureObjectParameter(UVs=NamedRerouteUsage()) {group=06_Emissive, parameterName=EmissiveCropped, sortPriority=1, texture=&T_BlacksRGB}
  n16 = VectorParameter() {
   group=06_Emissive
   parameterName=CroppedEmissiveUVs
   sortPriority=2
   channelNames {a=Bottom Right Corner V, b=Bottom Right Corner U, g=Upper Left Corner V, r=Upper Left Corner U}
   defaultValue=<rgba 0 0 1 1>
  }
  NamedRerouteDeclaration(None=MaterialFunctionCall!(
    UVs
    TextureIn=@n15
    UpperLeftCorner=ComponentMask(None=@n16) {g=true, r=true}
    LowerRightCorner=AppendVector(A=@n16.Bottom Right Corner U, B=@n16.Bottom Right Corner U)
    ExportFloat4
   ) {materialFunction=/Engine/Functions/Engine_MaterialFunctions01/Texturing/TextureCropping}.Cropped) {name=CroppedEmissive, nodeColor=<rgba 1 0.4688 0 1>}
  NamedRerouteDeclaration(None=FunctionInput() {bUsePreviewValueAsDefault=true, inputName=Alpha, sortPriority=3}) {name=Alpha, nodeColor=<rgba 1 0 0.1406 1>}
  NamedRerouteDeclaration(None=FunctionInput(Preview=TextureCoordinate()) {bUsePreviewValueAsDefault=true, inputName=Emissive UVs, inputType=FunctionInput_Vector2, sortPriority=1}) {name=UVs, nodeColor=<rgba 0 1 0.7188 1>}
 ]}
