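# Material function for the 2023 character parent material that optionally applies an opacity mask texture to incoming Material Attributes.
# A masks texture parameter supplies a selectable channel, red by default, with a static switch to invert it.
# That value multiplies the existing Opacity Mask attribute rather than replacing it.
# The EnableOpacityMap static switch passes the input through unchanged when off.
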MaterialFunction /Game/Global/Materials/Character_Parent_2023/Functions/MF_CharacterParent_2023_OpacityMask
graph {nodes [
  n1 = FunctionInput() {inputName=Result, inputType=FunctionInput_MaterialAttributes}
  n2 = GetMaterialAttributes(None=@n1) {attributeGetTypes [#1]}
  n3 = TextureSampleParameter2D() {
   group=03_Opacity Mask
   parameterName=OpacityMap
   samplerSource=SSM_Wrap_WorldGroupSettings
   samplerType=SAMPLERTYPE_Masks
   sortPriority=53
   texture=&T_WhiteMask
  }
  n4 = StaticComponentMaskParameter(None=@n3.RGBA) {defaultR=true, group=03_Opacity Mask, parameterName=OpacityMapChannel, sortPriority=57}
  n5 = Multiply(
   A=@n2.OpacityMask
   B=StaticSwitchParameter(True=OneMinus(None=@n4), False=@n4) {group=03_Opacity Mask, parameterName=InvertOpacityMap, sortPriority=60}
  )
  FunctionOutput(None=StaticSwitchParameter(
    True=SetMaterialAttributes!(MaterialAttributes=@n2.MaterialAttributes, OpacityMask=@n5) {attributeSetTypes [#1]}
    False=@n1
   ) {group=03_Opacity Mask, parameterName=EnableOpacityMap, sortPriority=50}) {outputName="Switch Param (False)"}
 ]}
